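# Material function for the customizable Zombie Elastic player body, recoloring the suit and skin of an incoming material.
# Pattern and mask textures drive primary/secondary colors and metallic, tinted skin, dirt, and adjusted normal, specular and roughness.
# Front and back stickers are placed by scaled UVs and blended as material attributes.
# View-space top shading darkens base color, while a colored directional/Fresnel rim feeds emissive, its exponent and strength varying with camera distance.
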MaterialFunction /Game/Characters/Player/Male/Medium/Bodies/M_MED_Zombie_Elastic/Materials/MF_ZombieElastic
graph {nodes [
  n1 = VertexColor()
  Add(
   A=SphereMask(A=@n1.R, B=ScalarParameter() {defaultValue=0.6095}) {attenuationRadius=0.01}
   B=SphereMask(A=@n1.R, B=ScalarParameter() {defaultValue=0.7, parameterName=Param2}) {attenuationRadius=0.01}
  )
  n2 = NamedRerouteUsage()
  n3 = NamedRerouteUsage()
  n4 = FunctionInput() {bUsePreviewValueAsDefault=true, inputName=Material, inputType=FunctionInput_MaterialAttributes}
  n5 = VectorParameter() {group=Elastic, parameterName=PrimaryColor, defaultValue=<rgba 0.05 0.05 0.9 1>}
  n6 = Multiply(A=@n5.RGB, B=NamedRerouteUsage()) {constB=0.8}
  n7 = VectorParameter() {group=Elastic, parameterName=SecondaryColor, defaultValue=<rgba 0.9 0.05 0.05 1>}
  n8 = SphereMask(A=Constant3Vector(), B=Subtract(A=@n5.RGB, B=@n7.RGB)) {attenuationRadius=0.1, hardnessPercent=0}
  n9 = Multiply(A=@n7.RGB, B=NamedRerouteUsage()) {constB=0}
  n10 = LinearInterpolate(
   A=Multiply(A=LinearInterpolate(A=@n5.RGB, B=@n6, Alpha=@n8)) {constB=1.05}
   B=Multiply(A=LinearInterpolate(A=@n7.RGB, B=@n9, Alpha=@n8)) {constB=0.85}
   Alpha=NamedRerouteUsage()
  )
  n11 = NamedRerouteUsage()
  n12 = TextureSampleParameter2D() {
   group=Elastic
   parameterName=Masks
   samplerType=SAMPLERTYPE_Masks
   sortPriority=0
   texture=/Game/Characters/Player/Male/Medium/Bodies/M_MED_Zombie_Elastic/Textures/Stages/1/T_M_Med_Zombie_Elastic_Stage_1_Body_M
  }
  n13 = NamedRerouteDeclaration(None=@n12.B) {name=SkinMask, nodeColor=<rgba 0 0.5547 1 1>}
  n14 = TextureSampleParameter2D() {
   group=Elastic
   parameterName=ZE_SecondaryNormal
   samplerType=SAMPLERTYPE_Normal
   sortPriority=0
   texture=/Game/Characters/Player/Male/Medium/Bodies/M_MED_Zombie_Elastic/Textures/Stages/1/T_M_Med_Zombie_Elastic_Stage_1_Body_Cloth_N
  }
  n15 = SetMaterialAttributes!(
   MaterialAttributes=@n4
   BaseColor=LinearInterpolate(A=LinearInterpolate(A=@n10, B=@n11, Alpha=@n13), B=Constant3Vector() {constant=<rgba 0.02 0.0084 0.0014 0>}, Alpha=NamedRerouteUsage())
   Normal=LinearInterpolate(A=NamedRerouteUsage(), B=@n14.RGB, Alpha=NamedRerouteUsage())
   Specular=NamedRerouteUsage()
   Roughness=NamedRerouteUsage()
   Metallic=LinearInterpolate(
    A=ScalarParameter() {group=Elastic, parameterName=EZ_PrimaryMetallic}
    B=ScalarParameter() {group=Elastic, parameterName=EZ_SecondaryMetallic}
    Alpha=NamedRerouteUsage()
   )
  ) {attributeSetTypes [
    #1
    #2
    #3
    #4
    #5
   ]}
  n16 = GetMaterialAttributes(None=BlendMaterialAttributes(A=@n15, B=NamedRerouteUsage(), Alpha=NamedRerouteUsage()) {vertexAttributeBlendType=UseA}) {attributeGetTypes [#1, #5]}
  FunctionOutput(None=SetMaterialAttributes!(MaterialAttributes=@n16.MaterialAttributes, BaseColor=NamedRerouteUsage(), EmissiveColor=Add(A=@n2, B=@n3)) {attributeSetTypes [#1, #6]})
  n17 = VectorParameter() {group=ElasticRim, parameterName=elasticPrimaryColoredRim, defaultValue=<rgba 1 1 1 1>}
  n18 = ScalarParameter() {defaultValue=0.33, group=ElasticRim, parameterName=elasticPrimaryColoredRimMult}
  n19 = VectorParameter() {group=ElasticRim, parameterName=elasticSecondaryColoredRim, defaultValue=<rgba 1 1 1 1>}
  n20 = ScalarParameter() {defaultValue=0.33, group=ElasticRim, parameterName=elasticSecondaryColoredRimMult}
  n21 = TextureSampleParameter2D() {
   group=Elastic
   parameterName=Pattern
   samplerType=SAMPLERTYPE_Masks
   texture=/Game/Characters/Player/Female/Medium/Bodies/F_MED_Elastic/Textures/T_F_MED_Elastic_Body_FX
  }
  n22 = Multiply(
   A=LinearInterpolate(A=Multiply(A=@n17.RGB, B=@n18), B=Multiply(A=@n19.RGB, B=@n20), Alpha=@n21.R)
   B=ScalarParameter() {defaultValue=0.333, group=ElasticRim, parameterName=elasticColoredRimMult}
  )
  n23 = ScalarParameter() {defaultValue=2.5, group=ElasticRim, parameterName=elastic_closeRimExponent}
  n24 = ScalarParameter() {defaultValue=1.5, group=ElasticRim, parameterName=elastic_farRimExponent}
  n25 = MaterialFunctionCall!(Blend Range=Constant() {desc=spread, r=3000}, Start Offset=Constant() {desc="start (was -1500)", r=-300}) {materialFunction=/Engine/Functions/Engine_MaterialFunctions03/Particles/Distance_Blend}
  n26 = PixelNormalWS()
  n27 = DotProduct(
   A=@n26
   B=Transform(None=VectorParameter() {group=ElasticRim, parameterName=elastic_RimDirection, defaultValue=<rgba 1 2 2 0>}.RGB) {transformSourceType=TRANSFORMSOURCE_View}
  )
  n28 = ScalarParameter() {defaultValue=0.45, group=ElasticRim, parameterName=elastic_ViewSpaceRim}
  n29 = DotProduct(A=@n26, B=CameraVectorWS())
  n30 = Add(
   A=Max(A=Multiply(A=@n27, B=@n28) {constB=0}) {constB=0}
   B=Multiply(A=OneMinus(None=@n29), B=ScalarParameter() {defaultValue=0.5, group=ElasticRim, parameterName=elastic_FresnelRimStrength})
  )
  n31 = Add(
   A=Power(Base=@n30, Exp=LinearInterpolate(A=@n23, B=@n24, Alpha=@n25.Result) {constA=2.5, desc=rim exponent - reduces with distance}) {constExponent=2.5}
   B=Multiply(A=Power(Base=@n30) {constExponent=1.5}) {constB=0.05}
  )
  n32 = ScalarParameter() {defaultValue=0.7, desc=Fresnel Tint Strength, group=ElasticRim, parameterName=colorRim_close, sliderMin=10}
  n33 = ScalarParameter() {defaultValue=2, desc=Fresnel Tint Strength, group=ElasticRim, parameterName=colorRim_far}
  NamedRerouteDeclaration(None=Multiply(A=Multiply(A=@n22, B=@n31), B=LinearInterpolate(A=@n32, B=@n33, Alpha=@n25.Result) {constA=1}) {constB=0.3, desc=white color fresnel}) {name=Emissive, nodeColor=<rgba 1 0.7734 0 1>}
  n34 = GetMaterialAttributes(None=@n4) {attributeGetTypes [
    #2
    #6
    #1
    #3
    #4
   ]}
  NamedRerouteDeclaration(None=@n34.EmissiveColor) {name=EmissiveInput, nodeColor=<rgba 0 0.0625 1 1>}
  NamedRerouteDeclaration(None=@n34.Normal) {name=NormalInput, nodeColor=<rgba 0 0.7187 1 1>}
  NamedRerouteDeclaration(None=@n34.Specular) {name=SpecularInput, nodeColor=<rgba 1 0 0.4219 1>}
  NamedRerouteDeclaration(None=@n34.Roughness) {name=RoughnessInput, nodeColor=<rgba 1 0 0.8203 1>}
  n35 = NamedRerouteUsage()
  n36 = LinearInterpolate(A=@n35, B=Multiply(A=@n35) {constB=0.5}, Alpha=NamedRerouteUsage()) {constB=0}
  n37 = Multiply(A=@n36) {constB=0.2}
  n38 = NamedRerouteUsage()
  NamedRerouteDeclaration(None=LinearInterpolate(A=LinearInterpolate(A=@n36, B=@n37, Alpha=@n38), B=NamedRerouteUsage(), Alpha=NamedRerouteUsage())) {name=SuitSpecular, nodeColor=<rgba 0 1 0.3672 1>}
  n39 = NamedRerouteUsage()
  NamedRerouteDeclaration(None=Multiply(A=Power(Base=@n39) {constExponent=0.5}, B=ScalarParameter() {defaultValue=3, group=Elastic, parameterName=EZ_DirtStrength}) {constB=3}) {name=DirtMask, nodeColor=<rgba 1 0 0.1641 1>}
  n40 = NamedRerouteUsage()
  n41 = LinearInterpolate(A=Multiply(A=@n40) {constB=0.75}, B=@n40, Alpha=NamedRerouteUsage()) {constB=0}
  n42 = Multiply(A=@n41) {constB=1.3}
  n43 = NamedRerouteUsage()
  NamedRerouteDeclaration(None=LinearInterpolate(A=LinearInterpolate(A=@n41, B=@n42, Alpha=@n43), B=NamedRerouteUsage(), Alpha=NamedRerouteUsage())) {name=SuitRoughness, nodeColor=<rgba 0.9922 1 0 1>}
  NamedRerouteDeclaration(None=@n12.G) {name=Ambocc, nodeColor=<rgba 1 0.0469 0 1>}
  n44 = VectorParameter() {group=Elastic, parameterName=StickerPosition, defaultValue=<rgba 0.1 -0.21 0 1>}
  n45 = VectorParameter() {group=Elastic, parameterName=StickerScale, defaultValue=<rgba -0.06 -0.06 0 1>}
  n46 = MaterialFunctionCall(UVs=Add(A=TextureCoordinate(), B=ComponentMask(None=@n44.RGB) {g=true, r=true}), Texture Scale=AppendVector(A=@n45.R, B=@n45.R)) {materialFunction=&ScaleUVsByCenter}
  n47 = TextureSampleParameter2D(UVs=@n46.UVs) {
   group=Elastic
   parameterName=Sticker
   samplerSource=SSM_Clamp_WorldGroupSettings
   texture=/Game/Global/Textures/Default/Blanks/T_Blank_White_ZeroAlpha_sRGBColor
  }
  n48 = VectorParameter() {group=Elastic, parameterName=BackStickerPosition, defaultValue=<rgba 0.35 0.4 0 1>}
  n49 = VectorParameter() {group=Elastic, parameterName=BackStickerScale, defaultValue=<rgba 0.07 0.07 0 1>}
  n50 = MaterialFunctionCall(UVs=Add(A=TextureCoordinate(), B=ComponentMask(None=@n48.RGB) {g=true, r=true}), Texture Scale=AppendVector(A=@n49.R, B=@n49.R)) {materialFunction=&ScaleUVsByCenter}
  n51 = TextureSampleParameter2D(UVs=@n50.UVs) {
   group=Elastic
   parameterName=Sticker
   samplerSource=SSM_Clamp_WorldGroupSettings
   texture=&T_Blank_White_ZeroAlpha_sRGBColor
  }
  n52 = Multiply(A=@n51.A, B=Clamp(None=@n50.0-1 mask))
  n53 = LinearInterpolate(A=@n47.RGB, B=@n51.RGB, Alpha=@n52)
  n54 = ScalarParameter() {defaultValue=1, group=Elastic, parameterName=ElasticStickerMult}
  n55 = NamedRerouteUsage()
  n56 = VectorParameter() {group=Elastic, parameterName=Sticker MSRE, defaultValue=<rgba 0 0.2 0.2 0.1>}
  n57 = SetMaterialAttributes!(
   MaterialAttributes=NamedRerouteUsage()
   BaseColor=LinearInterpolate(A=Multiply(A=@n53, B=@n54), B=Constant3Vector(), Alpha=Multiply(A=@n55) {constB=3})
   Metallic=@n56.R
   Specular=@n56.R
   Roughness=@n56.R
   Normal=NamedRerouteUsage()
  ) {attributeSetTypes [
    #1
    #5
    #3
    #4
    #2
   ]}
  NamedRerouteDeclaration(None=@n57) {name=StickerMaterial, nodeColor=<rgba 1 0 0.2344 1>}
  n58 = Multiply(A=@n47.A, B=Clamp(None=@n46.0-1 mask))
  n59 = NamedRerouteUsage()
  NamedRerouteDeclaration(None=Multiply(A=Add(A=@n58, B=@n52), B=OneMinus(None=@n59))) {name=StickerMask, nodeColor=<rgba 0.4766 0 1 1>}
  NamedRerouteDeclaration(None=@n34.MaterialAttributes) {name=MaterialInput, nodeColor=<rgba 0 1 0.2734 1>}
  n60 = VectorParameter() {group=ElasticRim, parameterName=elastic_TopLightDirection, defaultValue=<rgba 0 0.3 -0.1 0>}
  n61 = Add(A=DotProduct(A=PixelNormalWS(), B=Transform(None=@n60.RGB) {transformSourceType=TRANSFORMSOURCE_View})) {constB=0.5}
  n62 = Saturate(None=Power(Base=Multiply(A=@n61) {constB=1.4}))
  n63 = Max(
   A=@n16.BaseColor
   B=Multiply(A=VectorParameter() {group=Elastic, parameterName=elastic_minimumMetalicColor, defaultValue=<rgba 0.05 0.05 0.05 0>}.RGB, B=@n16.Metallic)
  )
  NamedRerouteDeclaration(None=LinearInterpolate(A=@n63, B=Multiply(A=@n63, B=@n62), Alpha=ScalarParameter() {defaultValue=0.5, group=ElasticRim, parameterName=elastic_ViewSpaceShading}) {constAlpha=0}) {name=Color, nodeColor=<rgba 0 0.9062 1 1>}
  NamedRerouteDeclaration(None=LinearInterpolate(
    A=ScalarParameter() {defaultValue=1, group=ElasticRim, parameterName=elastic_closeColorSeparationMult}
    B=ScalarParameter() {defaultValue=1.3, group=ElasticRim, parameterName=elastic_farColorSeparationMult}
    Alpha=@n25.Result
   )) {name=distanceColorSeparationMult, nodeColor=<rgba 0 1 0.0625 1>}
  NamedRerouteDeclaration(None=LinearInterpolate(
    A=ScalarParameter() {defaultValue=1, group=ElasticRim, parameterName=elastic_closeSecondaryColorSeparationMult}
    B=ScalarParameter() {defaultValue=0.2, group=ElasticRim, parameterName=elastic_farColorSecondarySeparationMult}
    Alpha=@n25.Result
   )) {name=distanceColorSecondarySeparationMult, nodeColor=<rgba 0 1 0.0625 1>}
  NamedRerouteDeclaration(None=@n34.BaseColor) {name=BaseColorInput, nodeColor=<rgba 1 0.6563 0 1>}
  NamedRerouteDeclaration(None=@n21.R) {name=PatternRed, nodeColor=<rgba 0 1 0.7422 1>}
  n64 = VectorParameter() {group=Elastic, parameterName=SkinColor, defaultValue=<rgba 0.1089 1 0.0894 0>}
  n65 = Desaturation(None=@n64.RGB, Fraction=ScalarParameter() {defaultValue=0.15, group=Elastic, parameterName=SkinColorDesaturation})
  n66 = NamedRerouteUsage()
  n67 = MaterialFunctionCall(Float3=MaterialFunctionCall!(In=@n66) {materialFunction=/Engine/Functions/Engine_MaterialFunctions01/Chromakeying/RGBtoHSV}.Result) {materialFunction=&BreakOutFloat3Components}
  n68 = ScalarParameter() {defaultValue=1.3, desc="Mults the V strength of the skin diffuse color. \r\nUsed by the tint color.", group=Elastic, parameterName=SkinVMult}
  n69 = ScalarParameter() {defaultValue=0.9, desc="0=No skin texture used. 1=Value of skin texture multed against skinColor", group=Elastic, parameterName=SkinVStength}
  n70 = LinearInterpolate(A=@n65, B=Multiply(A=Multiply(A=@n67.B, B=@n68), B=@n65), Alpha=@n69) {constAlpha=0.3}
  n71 = Multiply(A=@n64.A, B=ScalarParameter() {defaultValue=1, group=Elastic, parameterName=ColorSkinStrength}) {constB=0.5}
  NamedRerouteDeclaration(None=Multiply(A=LinearInterpolate(A=@n66, B=@n70, Alpha=@n71), B=NamedRerouteUsage())) {name=BaseColorTinted, nodeColor=<rgba 1 0.75 0 1>}
  NamedRerouteDeclaration(None=@n21.B) {name=PatternBlue, nodeColor=<rgba 0.3359 1 0 1>}
  Power(Base=@n12.G)
  Reroute(None=FunctionInput() {inputName=AO})
  ScalarParameter() {defaultValue=1, parameterName=Alpha}
 ]}
