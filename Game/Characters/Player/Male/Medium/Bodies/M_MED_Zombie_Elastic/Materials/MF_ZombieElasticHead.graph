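# Material function for the Elastic Zombie player head that recolors skin and adds tinted, emissive eyes to incoming material attributes.
# The base color's HSV value drives flesh and desaturated SkinColor tints, blended through masks and multiplied by the M texture's green channel.
# An eye layer with gradient color and its own surface response is mask-blended over the skin.
# Named reroutes declare masks from M/FX texture channels and vertex-red SphereMasks for eyes, teeth, gums and tongue.
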MaterialFunction /Game/Characters/Player/Male/Medium/Bodies/M_MED_Zombie_Elastic/Materials/MF_ZombieElasticHead
graph {
 properties {description=SkinColorSaturation}
 nodes [
  Constant3Vector()
  FunctionInput() {inputName=Skin, inputType=FunctionInput_Scalar}
  n1 = FunctionInput() {inputType=FunctionInput_MaterialAttributes}
  n2 = GetMaterialAttributes(None=@n1) {attributeGetTypes [#1]}
  n3 = MaterialFunctionCall(Float3=MaterialFunctionCall!(In=@n2.BaseColor) {materialFunction=/Engine/Functions/Engine_MaterialFunctions01/Chromakeying/RGBtoHSV}.Result) {materialFunction=&BreakOutFloat3Components}
  n4 = Multiply(A=VectorParameter() {parameterName=EZ_FleshColor, defaultValue=<rgba 1 0 0 0>}.RGB, B=@n3.B)
  n5 = ScalarParameter() {defaultValue=0.3, parameterName=EZ_FlashDesaturation}
  n6 = VectorParameter() {group=Elastic, parameterName=SkinColor, defaultValue=<rgba 0.1089 1 0.0894 0>}
  n7 = Desaturation(None=@n6.RGB, Fraction=ScalarParameter() {defaultValue=0.15, group=Elastic, parameterName=SkinColorDesaturation})
  n8 = ScalarParameter() {defaultValue=1.3, desc="Mults the V strength of the skin diffuse color. \r\nUsed by the tint color.", group=Elastic, parameterName=SkinVMult}
  n9 = Multiply(A=Multiply(A=@n3.B, B=@n8), B=@n7)
  n10 = ScalarParameter() {defaultValue=0.9, desc="0=No skin texture used. 1=Value of skin texture multed against skinColor", group=Elastic, parameterName=SkinVStength}
  n11 = Add(A=NamedRerouteUsage(), B=Multiply(A=NamedRerouteUsage()) {constB=0.9})
  n12 = ScalarParameter() {defaultValue=1, group=Elastic, parameterName=ColorSkinStrength}
  n13 = Subtract(A=Multiply(A=OneMinus(None=@n11), B=Multiply(A=@n6.A, B=@n12) {constB=0.5}) {constB=0.5}, B=NamedRerouteUsage())
  n14 = LinearInterpolate(
   A=@n2.BaseColor
   B=LinearInterpolate(A=Desaturation(None=@n4, Fraction=@n5), B=LinearInterpolate(A=@n7, B=@n9, Alpha=@n10) {constAlpha=0.3}, Alpha=Saturate(None=@n13))
   Alpha=NamedRerouteUsage()
  )
  n15 = TextureSampleParameter2D() {
   desc=red is eyes. Green is eye mask
   group=Elastic
   parameterName=M
   samplerType=SAMPLERTYPE_Masks
   sortPriority=0
   texture=/Game/Characters/Player/Male/Medium/Bodies/M_MED_Zombie_Elastic/Skins/CAU/Textures/T_M_MED_Zombie_Elastic_CAU_Head_M
  }
  n16 = VectorParameter() {group=Elastic, parameterName=EyeColor, defaultValue=<rgba 1 0.01 0 0>}
  n17 = Multiply(A=@n16.RGB) {constB=0.333}
  n18 = NamedRerouteUsage()
  n19 = VectorParameter() {group=Elastic, parameterName=Eye MSRE, defaultValue=<rgba 0 0.1 0.6 1>}
  n20 = Multiply(A=@n16.RGB) {constB=0.3}
  n21 = SetMaterialAttributes!(
   MaterialAttributes=@n1
   BaseColor=Multiply(A=LinearInterpolate(A=@n17, B=@n16.RGB, Alpha=@n18) {desc=gradient}) {constB=0.5}
   Metallic=@n19.R
   Specular=@n19.R
   Roughness=@n19.R
   EmissiveColor=Multiply(A=LinearInterpolate(A=@n20, B=@n16.RGB, Alpha=@n18) {desc=gradient}) {constB=0.2}
  ) {attributeSetTypes [
    #1
    #2
    #3
    #4
    #5
   ]}
  n22 = BlendMaterialAttributes(
   A=SetMaterialAttributes!(MaterialAttributes=@n1, BaseColor=Multiply(A=@n14, B=@n15.G)) {attributeSetTypes [#1]}
   B=@n21
   Alpha=NamedRerouteUsage()
  )
  FunctionOutput(None=@n22)
  Multiply(
   A=LinearInterpolate(A=@n16.A, Alpha=ScalarParameter() {group=Elastic, parameterName=EyeColorAlpha1})
   B=OneMinus(None=ScalarParameter() {group=Elastic, parameterName=ElasticVisor})
  )
  NamedRerouteDeclaration(None=SphereMask(A=VertexColor().R, B=Constant() {r=0.1}) {attenuationRadius=0.01}) {name=eyeMask, nodeColor=<rgba 0 1 0.7422 1>}
  NamedRerouteDeclaration(None=NamedRerouteUsage()) {name=eyeSphereMask, nodeColor=<rgba 1 0 0.0937 1>}
  n23 = NamedRerouteUsage()
  n24 = NamedRerouteUsage()
  NamedRerouteDeclaration(None=Saturate(None=Add(A=@n23, B=@n24))) {name=M_Skin, nodeColor=<rgba 0.9219 0 1 1>}
  NamedRerouteDeclaration(None=SphereMask(A=VertexColor().R, B=ScalarParameter() {defaultValue=0.2}) {attenuationRadius=0.1}) {name=TeethAndGums, nodeColor=<rgba 0 1 0.8828 1>}
  NamedRerouteDeclaration(None=@n15.B) {name=M_Blue, nodeColor=<rgba 1 0.8438 0 1>}
  n25 = TextureSampleParameter2D() {
   desc=red is eyes. Green is eye mask
   group=Elastic
   parameterName=FX
   sortPriority=0
   texture=/Game/Characters/Player/Male/Medium/Bodies/M_MED_Zombie_Elastic/Skins/CAU/Textures/T_M_MED_Zombie_Elastic_CAU_Head_FX
  }
  NamedRerouteDeclaration(None=@n25.G) {name=FX_G_gums, nodeColor=<rgba 0 0.2734 1 1>}
  NamedRerouteDeclaration(None=@n14) {name=Preview, nodeColor=<rgba 0.7813 0 1 1>}
  NamedRerouteDeclaration(None=@n25.R) {name=FX_R_flesh, nodeColor=<rgba 0.1953 0 1 1>}
  NamedRerouteDeclaration(None=@n25.G) {name=FX_G_eyes, nodeColor=<rgba 0.0781 1 0 1>}
  NamedRerouteDeclaration(None=SphereMask(A=VertexColor().R, B=Constant() {r=0.3}) {attenuationRadius=0.025}) {name=TongueMask, nodeColor=<rgba 1 0.6094 0 1>}
  OneMinus(None=NamedRerouteUsage())
  Power() {constExponent=1}
  SetMaterialAttributes!(MaterialAttributes=@n22, Normal, BaseColor=NamedRerouteUsage(), WorldPositionOffset) {attributeSetTypes [#6, #1, #7]}
  SphereMask(A=TextureCoordinate(), B=ComponentMask(None=VectorParameter() {defaultValue=<rgba 0.8876 0.5524 0 1>}.RGB) {g=true, r=true}) {attenuationRadius=0.1, hardnessPercent=0}
 ]
}
